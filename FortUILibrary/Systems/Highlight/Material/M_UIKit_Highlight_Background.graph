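# Translucent UI material for the background of UIKit highlight elements, with a color style chosen by the HighlightType parameter.
# A Switch selects a fixed color for each named type, except Quiet, which uses the CoreUI BackplateC function fed by Slate post buffer 0.
# A CoreUI rounded mask builds the shape from per-corner radius and skew parameters.
# Its width shrinks by the skew tangent times aspect ratio, alongside X/Y scale parameters.
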Material /FortUILibrary/Systems/Highlight/Material/M_UIKit_Highlight_Background
graph {
 properties {
  bUsedWithStaticMesh=false
  blendMode=BLEND_Translucent
  materialDomain=MD_UI
  pixelDepthOffsetMode=PDOM_Legacy
  usedShadingModels=Default Lit
 }
 nodes [
  n1 = ScalarParameter() {group=Layout, parameterName=SkewAngle, sortPriority=4}
  n2 = MaterialFunctionCall!(Dimensions=MaterialFunctionCall() {materialFunction=&GetUserInterfaceUV}.Pixel Size) {materialFunction=/FortUILibrary/MaterialFunctions/Basics/MF_UI_GetAspectRatio}
  n3 = OneMinus(None=Multiply(A=Tangent!(None=@n1) {period=360}, B=ComponentMask(None=@n2.V Ratio) {g=true}))
  n4 = Constant() {r=1}
  n5 = ScalarParameter() {defaultValue=1, group=Layout, parameterName=ScaleX, sortPriority=1}
  n6 = ScalarParameter() {defaultValue=1, group=Layout, parameterName=ScaleY, sortPriority=1}
  MaterialFunctionCall(
   Size=Multiply(
    A=AppendVector(A=@n3, B=@n4)
    B=MaterialFunctionCall(X=@n5, Y=@n6) {materialFunction=&MakeFloat2}.Result
   )
   Corner Radius=MaterialFunctionCall(
    X=ScalarParameter() {defaultValue=0.5, group=Layout, parameterName=TopLeft, sortPriority=10}
    Y=ScalarParameter() {defaultValue=0.5, group=Layout, parameterName=BottomLeft, sortPriority=13}
    Z=ScalarParameter() {defaultValue=0.5, group=Layout, parameterName=TopRight, sortPriority=11}
    A=ScalarParameter() {defaultValue=0.5, group=Layout, parameterName=BottomRight, sortPriority=12}
   ) {materialFunction=&MakeFloat4}.Result
   Radius in Pixels=ScalarParameter() {group=Layout, parameterName=Radius in Pixels, sortPriority=9}
   Skew Angle=ScalarParameter() {group=Layout, parameterName=SkewAngle, sortPriority=4}
  ) {materialFunction=&MF_UI_CoreUI_RoundedMask}
  MaterialExpressionConstant3Vector_3 = Constant3Vector() {constant=<rgba 0.9301 1 0.0103 1>}
  MaterialExpressionConstant3Vector_2 = Constant3Vector() {constant=<rgba 0.0395 1 0.2307 1>}
  MaterialExpressionConstant3Vector_1 = Constant3Vector() {constant=<rgba 0.0048 0.004 0.0116 1>}
  MaterialExpressionConstant3Vector_0 = Constant3Vector() {constant=<rgba 1 1 1 0>}
  MaterialExpressionMaterialFunctionCall_1 = MaterialFunctionCall!(Post Buffer Blur=MaterialFunctionCall!("UVs (Optional)") {materialFunction=/Engine/Functions/UserInterface/GetSlatePost0}.LinearRGB) {materialFunction=/FortUILibrary/MaterialFunctions/CoreUI/MF_UI_CoreUI_BackplateC}
  MaterialExpressionConstant3Vector_4 = Constant3Vector() {constant=<rgba 0.6706 0 0.0039 1>}
  MaterialExpressionConstant3Vector_5 = Constant3Vector() {constant=<rgba 1 0.063 0.3372 1>}
  MaterialExpressionConstant3Vector_6 = Constant3Vector() {constant=<rgba 0 1 0.6038 1>}
  Switch!(
   SwitchValue=ScalarParameter() {parameterName=HighlightType}
   Default
   Attention=@MaterialExpressionConstant3Vector_3
   Social=@MaterialExpressionConstant3Vector_2
   Utility=@MaterialExpressionConstant3Vector_1
   Deals=@MaterialExpressionConstant3Vector_0
   Quiet=@MaterialExpressionMaterialFunctionCall_1.Result
   Live=@MaterialExpressionConstant3Vector_4
   Urgent=@MaterialExpressionConstant3Vector_5
   LiveAlt=@MaterialExpressionConstant3Vector_6
  ) {inputs [
    {
     inputName=Attention
     input {
      expression=&self:MaterialExpressionConstant3Vector_3
      inputName=None
      mask=1
      maskA=0
      maskB=1
      maskG=1
      maskR=1
      outputIndex=0
     }
    }
    {
     inputName=Social
     input {
      expression=&self:MaterialExpressionConstant3Vector_2
      inputName=None
      mask=1
      maskA=0
      maskB=1
      maskG=1
      maskR=1
      outputIndex=0
     }
    }
    {
     inputName=Utility
     input {
      expression=&self:MaterialExpressionConstant3Vector_1
      inputName=None
      mask=1
      maskA=0
      maskB=1
      maskG=1
      maskR=1
      outputIndex=0
     }
    }
    {
     inputName=Deals
     input {
      expression=&self:MaterialExpressionConstant3Vector_0
      inputName=None
      mask=1
      maskA=0
      maskB=1
      maskG=1
      maskR=1
      outputIndex=0
     }
    }
    {
     inputName=Quiet
     input {
      expression=&self:MaterialExpressionMaterialFunctionCall_1
      inputName=None
      mask=0
      maskA=0
      maskB=0
      maskG=0
      maskR=0
      outputIndex=0
     }
    }
    {
     inputName=Live
     input {
      expression=&self:MaterialExpressionConstant3Vector_4
      inputName=None
      mask=1
      maskA=0
      maskB=1
      maskG=1
      maskR=1
      outputIndex=0
     }
    }
    {
     inputName=Urgent
     input {
      expression=&self:MaterialExpressionConstant3Vector_5
      inputName=None
      mask=1
      maskA=0
      maskB=1
      maskG=1
      maskR=1
      outputIndex=0
     }
    }
    {
     inputName=LiveAlt
     input {
      expression=&self:MaterialExpressionConstant3Vector_6
      inputName=None
      mask=1
      maskA=0
      maskB=1
      maskG=1
      maskR=1
      outputIndex=0
     }
    }
   ]}
 ]
}
