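# Base UI Kit highlight widget: a short uppercase label over a material background, styled by type and size.
# PreConstruct applies type, size and width limits, then sets the text; On Widget Styled reapplies all but the text.
# Type picks the text color and drives a HighlightType scalar on the background's dynamic material; size picks font, padding and minimum height.
# Empty text collapses visibility; width limits go to the root SizeBox or are cleared.
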WidgetBlueprint /FortUILibrary/Systems/Highlight/WBP_UIKit_Highlight_Base
graphs [
 EventGraph {nodes [
   Event PreConstruct()
   > InitializeHighlight()
   > SetText!(execute, self:Self Object Reference, Text:Text=Get HighlightText() -> (HighlightText:Text)) -> (then)
   Event On Widget Styled() {is=title K2Node_Event}
   > InitializeHighlight()
  ]}
 InitializeHighlight {nodes [
   InitializeHighlight() {is=title K2Node_FunctionEntry}
   > SetType!(execute, self, Type:E_UI_HighlightType Enum=Get HighlightType() -> (HighlightType:E_UI_HighlightType Enum=NewEnumerator3)) -> (then)
   > SetHighlightSize!(execute, self, FontSize:E_UI_HighlightSize Enum=Get HighlightSize() -> (HighlightSize:E_UI_HighlightSize Enum=NewEnumerator0)) -> (then)
   > SetWidth!(execute, self, NewWidthOverride:double=Get WidthOverride() -> (WidthOverride:double)) -> (then)
   > SetMaxWidth!(execute, self, NewMaxWidth:double=Get MaxWidth() -> (MaxWidth:double)) -> (then)
  ]}
 SetBackgroundStyle {nodes [
   n1 = SetBackgroundStyle!() -> (then, HighlightType:E_UI_HighlightType Enum=NewEnumerator0) {is=title K2Node_FunctionEntry}
   > n2 = GetDynamicMaterial(self=Get Background() -> (Background:Image Object Reference))
   > SetScalarParameterValue(
    self=@n2
    ParameterName=HighlightType
    Value=Select!(
     NewEnumerator3:float=2
     NewEnumerator0:float=0
     NewEnumerator1:float=1
     NewEnumerator4:float=3
     NewEnumerator5:float=4
     NewEnumerator6:float=5
     NewEnumerator7:float=6
     NewEnumerator8:float=7
     Index:E_UI_HighlightType Enum=@n1
    ) -> (ReturnValue:float)
   )
  ]}
 SetHighlightSize {nodes [
   n1 = SetHighlightSize!() -> (then, FontSize:E_UI_HighlightSize Enum=NewEnumerator1) {is=title K2Node_FunctionEntry}
   > n2 = Set HighlightSize(HighlightSize:E_UI_HighlightSize Enum=@n1) -> (Output_Get:E_UI_HighlightSize Enum=NewEnumerator0)
   > Set Font!(
    execute
    Font:Slate Font Info Structure=Select!(
     NewEnumerator0:Slate Font Info Structure=Get HighlightFont_Compact() -> (HighlightFont_Compact:Slate Font Info Structure)
     NewEnumerator1:Slate Font Info Structure=Get HighlightFont_Default() -> (HighlightFont_Default:Slate Font Info Structure)
     Index:E_UI_HighlightSize Enum=@n2
    ) -> (ReturnValue:Slate Font Info Structure)
    self:Text Object Reference=Get Text_Highlight() -> (Text_Highlight:Common Text Object Reference)
   ) -> (then, Output_Get:Slate Font Info Structure)
   > Set Padding!(
    execute
    Padding:Margin Structure=Select(
     NewEnumerator0:Margin Structure=Get HighlightPadding_Compact() -> (HighlightPadding_Compact:Margin Structure)
     NewEnumerator1:Margin Structure=Get HighlightPadding_Default() -> (HighlightPadding_Default:Margin Structure)
     Index=Get HighlightSize() -> (HighlightSize=NewEnumerator0)
    ) -> (ReturnValue:Margin Structure)
    self:Grid Slot Object Reference=SlotAsGridSlot(Widget=Get Text_Highlight())
   ) -> (then, Output_Get:Margin Structure)
   > Set MinDesiredHeight!(
    execute
    MinDesiredHeight:float=Select(NewEnumerator0:float=24, NewEnumerator1:float=32, Index=Get HighlightSize() -> (HighlightSize=NewEnumerator0)) -> (ReturnValue:float)
    self:Size Box Object Reference=Get SB_WidthLimit() -> (SB_WidthLimit:Size Box Object Reference)
   ) -> (then, Output_Get:float)
  ]}
 SetMaxWidth {nodes [
   n1 = SetMaxWidth!() -> (then, NewMaxWidth:double=~) {is=title K2Node_FunctionEntry}
   > n2 = Set MaxWidth(MaxWidth:double=@n1) -> (Output_Get:double)
   > n3 = Branch(Condition="float > float"(A=@n2))
   n4 = Get SB_WidthLimit()
   > SetMaxDesiredWidth!(execute, self:Size Box Object Reference=@n4, InMaxDesiredWidth:float=@n2) -> (then)
   ClearMaxDesiredWidth(execute=@n3.else, self:Size Box Object Reference=@n4)
  ]}
 SetText {nodes [
   n1 = SetText!() -> (then, Text:Text) {is=title K2Node_FunctionEntry}
   > n2 = Set HighlightText(HighlightText:Text=@n1) -> (Output_Get:Text)
   > n3 = Set Text(Text=@n2, self=Get Text_Highlight())
   > Set Visibility(Visibility=Select!(Option 0:ESlateVisibility Enum=SelfHitTestInvisible, Option 1:ESlateVisibility Enum=Collapsed, Index:bool=TextIsEmpty(InText=@n3)) -> (ReturnValue:ESlateVisibility Enum))
  ]}
 SetTextColor {nodes [
   n1 = SetTextColor!() -> (then, Color:Linear Color Structure) {is=title K2Node_FunctionEntry}
   > Set ColorAndOpacity!(execute, ColorAndOpacity:Slate Color Structure=Make Slate Color(SpecifiedColor=@n1), self:Text Object Reference=Get Text_Highlight()) -> (then, Output_Get:Slate Color Structure)
  ]}
 SetType {nodes [
   n1 = SetType!() -> (then, Type:E_UI_HighlightType Enum=NewEnumerator0) {is=title K2Node_FunctionEntry}
   > Set HighlightType(HighlightType:E_UI_HighlightType Enum=@n1) -> (Output_Get:E_UI_HighlightType Enum=NewEnumerator3)
   n2 = Get!() -> (
    HighlightColors_AttentionTextColor_11_#1:Linear Color Structure
    HighlightColors_SocialTextColor_16_#2:Linear Color Structure
    HighlightColors_UtilityTextColor_22_#3:Linear Color Structure
    HighlightColors_DealsTextColor_24_#4:Linear Color Structure
    HighlightColors_QuietTextColor_31_#5:Linear Color Structure
    HighlightColors_LiveTextColor_34_#6:Linear Color Structure
    HighlightColors_UrgentTextColor_37_#7:Linear Color Structure
    HighlightColors_LiveAltTextColor_40_#8:Linear Color Structure
   )
   > SetTextColor!(
    execute
    self
    Color:Linear Color Structure=Select!(
     NewEnumerator3:Linear Color Structure=@n2.HighlightColors_UtilityTextColor_22_#3
     NewEnumerator0:Linear Color Structure=@n2
     NewEnumerator1:Linear Color Structure=@n2.HighlightColors_SocialTextColor_16_#2
     NewEnumerator4:Linear Color Structure=@n2.HighlightColors_DealsTextColor_24_#4
     NewEnumerator5:Linear Color Structure=@n2.HighlightColors_QuietTextColor_31_#5
     NewEnumerator6:Linear Color Structure=@n2.HighlightColors_LiveTextColor_34_#6
     NewEnumerator7:Linear Color Structure=@n2.HighlightColors_UrgentTextColor_37_#7
     NewEnumerator8:Linear Color Structure=@n2.HighlightColors_LiveAltTextColor_40_#8
     Index:E_UI_HighlightType Enum=Get HighlightType() -> (HighlightType=NewEnumerator3)
    ) -> (ReturnValue:Linear Color Structure)
   ) -> (then) {is=title K2Node_CallFunction}
   > SetBackgroundStyle!(execute, self, HighlightType:E_UI_HighlightType Enum=Get HighlightType() -> (HighlightType=NewEnumerator3)) -> (then) {is=title K2Node_CallFunction}
   > SetTextTransformPolicy!(execute, self:Text Object Reference=Get Text_Highlight(), InTransformPolicy:ETextTransformPolicy Enum=ToUpper) -> (then)
  ]}
 SetWidth {nodes [
   n1 = SetWidth!() -> (then, NewWidthOverride:double=~) {is=title K2Node_FunctionEntry}
   > n2 = Set WidthOverride(WidthOverride:double=@n1) -> (Output_Get:double)
   > n3 = Branch(Condition="float > float"(A=@n2))
   n4 = Get SB_WidthLimit()
   > SetWidthOverride(self=@n4, InWidthOverride=@n2)
   ClearWidthOverride(execute=@n3.else, self:Size Box Object Reference=@n4)
  ]}
]
variables [
 HighlightColors {
  type=S_UI_HighlightColors
  default {
   attentionTextColor=<rgba 0.0048 0.004 0.0116 1>
   dealsTextColor=<rgba 0.0048 0.004 0.0116 1>
   liveAltTextColor=<rgba 0.0052 0.0052 0.007 1>
   liveTextColor=<rgba 1 1 1 1>
   quietTextColor=<rgba 1 1 1 1>
   socialTextColor=<rgba 0 0.0331 0.0296 1>
   urgentTextColor=<rgba 0.0144 0.0033 0.0203 1>
   utilityTextColor=<rgba 1 1 1 1>
  }
 }
 HighlightFont_Compact {
  type=SlateFontInfo
  default {
   bForceMonospaced=false
   bMaterialIsStencil=false
   fontMaterial=null
   fontObject=&HeadingNow
   letterSpacing=10
   monospacedWidth=1
   size=12.75
   skewAmount=0
   typefaceFontName=76 Bold Italic
   outlineSettings {
    bApplyOutlineToDropShadows=false
    bMiteredCorners=false
    bSeparateFillAlpha=false
    outlineMaterial=null
    outlineSize=0
    outlineColor=<rgba 0 0 0 1>
   }
  }
 }
 HighlightFont_Default {
  type=SlateFontInfo
  default {
   bForceMonospaced=false
   bMaterialIsStencil=false
   fontMaterial=null
   fontObject=&HeadingNow
   letterSpacing=10
   monospacedWidth=1
   size=15.75
   skewAmount=0
   typefaceFontName=86 Bold Italic
   outlineSettings {
    bApplyOutlineToDropShadows=false
    bMiteredCorners=false
    bSeparateFillAlpha=false
    outlineMaterial=null
    outlineSize=0
    outlineColor=<rgba 0 0 0 1>
   }
  }
 }
 HighlightPadding_Compact {type=Margin, default=<ltrb 8 0 8 0>}
 HighlightPadding_Default {type=Margin, default=<ltrb 12 0 12 0>}
 HighlightSize {type=E_UI_HighlightSize, default=Default}
 HighlightText {type=string, default=New}
 HighlightType {type=E_UI_HighlightType, default=Live}
 MaxWidth {type=number}
 WidthOverride {type=number}
]
widgets {
 info {inheritedWidgetCount=0, parentClass=&UserWidget, rootWidgetClass=&SizeBox, widgetCount=4}
 widgets [
  SB_WidthLimit {
   widgetClassPath=&SizeBox
   bIsVariable=true
   widget=SB_WidthLimit
   properties {heightOverride=32, maxDesiredHeight=24, minDesiredHeight=24}
  }
  Root {widgetClassPath=&GridPanel, parent=SB_WidthLimit, widget=Root, properties {columnFill [1], rowFill [1]}, slot {class=SizeBoxSlot}}
  Background {
   widgetClassPath=&Image
   bIsVariable=true
   parent=Root
   widget=Background
   properties {
    visibility=HitTestInvisible
    brush {
     bIsDynamicallyLoaded=false
     drawAs=Image
     imageType=NoImage
     mirroring=NoMirror
     resourceName=None
     resourceObject=/FortUILibrary/Systems/Highlight/MaterialInstances/MI_UIKit_HighlightBackground
     tiling=NoTile
     imageSize=<0 0>
     margin=<ltrb 0 0 0 0>
     outlineSettings {
      bUseBrushTransparency=false
      roundingType=HalfHeightRadius
      width=0
      color {colorUseRule=UseColor_Specified, specifiedColor=<rgba 0 0 0 0>}
      cornerRadii=<0 0 0 0>
     }
     tintColor {colorUseRule=UseColor_Specified, specifiedColor=<rgba 1 1 1 1>}
     uVRegion {bIsValid=false, max=<0 0>, min=<0 0>}
    }
   }
   slot {class=GridSlot}
  }
  Text_Highlight {
   widgetClassPath=&CommonTextBlock
   bIsVariable=true
   parent=Root
   widget=Text_Highlight
   properties {
    applyLineHeightToBottomLine=false
    bAutoCollapseWithEmptyText=true
    justification=Center
    mobileFontSizeMultiplier=1.14
    scrollStyle=&TextScrollStyle-Base_C
    text=New
    colorAndOpacity {colorUseRule=UseColor_Specified, specifiedColor=<rgba 1 0 0.9536 1>}
    font {
     bForceMonospaced=false
     bMaterialIsStencil=false
     fontMaterial=null
     fontObject=&HeadingNow
     letterSpacing=0
     monospacedWidth=1
     size=15.75
     skewAmount=0
     typefaceFontName=86 Bold Italic
     outlineSettings {
      bApplyOutlineToDropShadows=false
      bMiteredCorners=false
      bSeparateFillAlpha=false
      outlineMaterial=null
      outlineSize=0
      outlineColor=<rgba 0 0 0 1>
     }
    }
    shadowOffset=<0 0>
   }
   slot {class=GridSlot, properties {horizontalAlignment=HAlign_Center, layer=1, verticalAlignment=VAlign_Center, padding=<ltrb 12 0 12 0>}}
  }
 ]
}
